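# Physics asset for the FaceAcc part of the female medium CopperMoggy SwanHair player cosmetic, defining physics for its hair bones.
# Each bone has a single capsule body, with a small one on the head as the root.
# Back, left and right master bones hang off the head with tight angular limits.
# Numbered hair bones constrain directly to their group's master rather than chaining, each with individually tuned swing and twist limits.
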PhysicsAsset /BRCosmetics/Characters/Player/Female/Medium/Bodies/F_MED_CopperMoggy_SwanHair/Meshes/Parts/F_MED_CopperMoggy_SwanHair_FaceAcc_Physics
bodies {
 dyn_hair_01_b [{shapeType=Capsule, length=2.1642, radius=1.8526, shapeName=dyn_hair_01_b_capsule, center=<0.3 0 0>, rotation=<rot 90 0 0>}]
 dyn_hair_01_l [{shapeType=Capsule, length=3.2196, radius=2.1672, shapeName=dyn_hair_01_l_capsule, center=<0.3 0 0>, rotation=<rot 90 0 0>}]
 dyn_hair_01_r [{shapeType=Capsule, length=3.7999, radius=1.6887, shapeName=dyn_hair_01_r_capsule, center=<0.3 0 0>, rotation=<rot 90 0 0>}]
 dyn_hair_02_b [{shapeType=Capsule, length=1.8441, radius=1.8593, shapeName=dyn_hair_02_b_capsule, center=<0.37 0 0>, rotation=<rot 90 0 0>}]
 dyn_hair_02_l [{shapeType=Capsule, length=3.2468, radius=1.8221, shapeName=dyn_hair_02_l_capsule, center=<0.3 0 0>, rotation=<rot -0.0079 -90.0126 -82.8163>}]
 dyn_hair_02_r [{shapeType=Capsule, length=3.3638, radius=2.0665, shapeName=dyn_hair_02_r_capsule, center=<0.3 0 0>, rotation=<rot 90 0 0>}]
 dyn_hair_03_b [{shapeType=Capsule, length=1.1418, radius=1.7939, shapeName=dyn_hair_03_b_capsule, center=<0.37 0 0>, rotation=<rot 90 0 0>}]
 dyn_hair_03_l [{shapeType=Capsule, length=2.7755, radius=1.4208, shapeName=dyn_hair_03_l_capsule, center=<0.3 0 0>, rotation=<rot 0 -90.0001 -70.1007>}]
 dyn_hair_03_r [{shapeType=Capsule, length=4.3426, radius=2.1089, shapeName=dyn_hair_03_r_capsule, center=<0.3 0 0>, rotation=<rot 88.009 -0.0001 -0.0001>}]
 dyn_hair_04_b [{shapeType=Capsule, length=1.3462, radius=2.508, shapeName=dyn_hair_04_b_capsule, center=<0.37 0 0>, rotation=<rot 90 0 0>}]
 dyn_hair_04_l [{shapeType=Capsule, length=2.0127, radius=1.6579, shapeName=dyn_hair_04_l_capsule, center=<0.3 0 0>, rotation=<rot 90 0 0>}]
 dyn_hair_04_r [{shapeType=Capsule, length=3.2029, radius=2.4138, shapeName=dyn_hair_04_r_capsule, center=<0.3 0 0>, rotation=<rot 0 -93.8918 -86.0647>}]
 dyn_hair_05_b [{shapeType=Capsule, length=2.6547, radius=2.3419, shapeName=dyn_hair_05_b_capsule, center=<0.37 0 0>, rotation=<rot 90 0 0>}]
 dyn_hair_05_l [{shapeType=Capsule, length=3.8527, radius=1.8157, shapeName=dyn_hair_05_l_capsule, center=<0.3 0 0>, rotation=<rot 90 0 0>}]
 dyn_hair_05_r [{shapeType=Capsule, length=2.7474, radius=1.7984, shapeName=dyn_hair_05_r_capsule, center=<0.3 0 0>, rotation=<rot 0 -89.9999 90.314>}]
 dyn_hair_06_b [{shapeType=Capsule, length=2.5995, radius=1.9198, shapeName=dyn_hair_06_b_capsule, center=<0.37 0 0>, rotation=<rot 90 0 0>}]
 dyn_hair_06_l [{shapeType=Capsule, length=2.3459, radius=1.7657, shapeName=dyn_hair_06_l_capsule, center=<0.3 0 0>, rotation=<rot 90 0 0>}]
 dyn_hair_07_l [{shapeType=Capsule, length=1.9065, radius=1.7642, shapeName=dyn_hair_07_l_capsule, center=<0.3 0 0>, rotation=<rot 90 0 0>}]
 dyn_hair_08_l [{shapeType=Capsule, length=1.477, radius=1.6954, shapeName=dyn_hair_08_l_capsule, center=<0.3 0 0>, rotation=<rot 90 0 0>}]
 dyn_hair_master_b [{shapeType=Capsule, length=2, radius=2, shapeName=dyn_hair_master_b_capsule, center=<1 0 0>, rotation=<rot 89 0 0>}]
 dyn_hair_master_l [{shapeType=Capsule, length=2, radius=2, shapeName=dyn_hair_master_l_capsule, center=<1 0 0>, rotation=<rot 90 0 0>}]
 dyn_hair_master_r [{shapeType=Capsule, length=2, radius=2, shapeName=dyn_hair_master_r_capsule, center=<1 0 0>, rotation=<rot 90 0 0>}]
 head [{shapeType=Capsule, length=0.5, radius=0.5, shapeName=head_capsule}]
}
constraints [
 {bone1Name=dyn_hair_master_b, bone2Name=head, swing1LimitDegrees=10, swing2LimitDegrees=10, twistLimitDegrees=2}
 {bone1Name=dyn_hair_01_b, bone2Name=dyn_hair_master_b, swing1LimitDegrees=30, swing2LimitDegrees=15, twistLimitDegrees=6}
 {bone1Name=dyn_hair_02_b, bone2Name=dyn_hair_master_b, swing1LimitDegrees=25, swing2LimitDegrees=20, twistLimitDegrees=2}
 {bone1Name=dyn_hair_03_b, bone2Name=dyn_hair_master_b, swing1LimitDegrees=35, swing2LimitDegrees=15, twistLimitDegrees=10}
 {bone1Name=dyn_hair_04_b, bone2Name=dyn_hair_master_b, swing1LimitDegrees=30, swing2LimitDegrees=15, twistLimitDegrees=10}
 {bone1Name=dyn_hair_05_b, bone2Name=dyn_hair_master_b, swing1LimitDegrees=35, swing2LimitDegrees=20, twistLimitDegrees=4}
 {bone1Name=dyn_hair_06_b, bone2Name=dyn_hair_master_b, swing1LimitDegrees=30, swing2LimitDegrees=20, twistLimitDegrees=2}
 {bone1Name=dyn_hair_master_l, bone2Name=head, swing1LimitDegrees=2, swing2LimitDegrees=10, twistLimitDegrees=10}
 {bone1Name=dyn_hair_01_l, bone2Name=dyn_hair_master_l, swing1LimitDegrees=15, swing2LimitDegrees=15, twistLimitDegrees=2}
 {bone1Name=dyn_hair_02_l, bone2Name=dyn_hair_master_l, swing1LimitDegrees=15, swing2LimitDegrees=15, twistLimitDegrees=2}
 {bone1Name=dyn_hair_03_l, bone2Name=dyn_hair_master_l, swing1LimitDegrees=34, swing2LimitDegrees=8, twistLimitDegrees=2}
 {bone1Name=dyn_hair_04_l, bone2Name=dyn_hair_master_l, swing1LimitDegrees=15, swing2LimitDegrees=15, twistLimitDegrees=2}
 {bone1Name=dyn_hair_05_l, bone2Name=dyn_hair_master_l, swing1LimitDegrees=15, swing2LimitDegrees=8, twistLimitDegrees=3}
 {bone1Name=dyn_hair_06_l, bone2Name=dyn_hair_master_l, swing1LimitDegrees=46, swing2LimitDegrees=20, twistLimitDegrees=7}
 {bone1Name=dyn_hair_07_l, bone2Name=dyn_hair_master_l, swing1LimitDegrees=60, swing2LimitDegrees=60, twistLimitDegrees=10}
 {bone1Name=dyn_hair_08_l, bone2Name=dyn_hair_master_l, swing1LimitDegrees=26, swing2LimitDegrees=11, twistLimitDegrees=11}
 {bone1Name=dyn_hair_master_r, bone2Name=head, swing1LimitDegrees=2, swing2LimitDegrees=10, twistLimitDegrees=10}
 {bone1Name=dyn_hair_01_r, bone2Name=dyn_hair_master_r, swing1LimitDegrees=15, swing2LimitDegrees=5, twistLimitDegrees=2}
 {bone1Name=dyn_hair_02_r, bone2Name=dyn_hair_master_r, swing1LimitDegrees=14, swing2LimitDegrees=14, twistLimitDegrees=2}
 {bone1Name=dyn_hair_03_r, bone2Name=dyn_hair_master_r, swing1LimitDegrees=23.64, swing2LimitDegrees=15, twistLimitDegrees=5}
 {bone1Name=dyn_hair_04_r, bone2Name=dyn_hair_master_r, swing1LimitDegrees=20, swing2LimitDegrees=5, twistLimitDegrees=3}
 {bone1Name=dyn_hair_05_r, bone2Name=dyn_hair_master_r, swing1LimitDegrees=25, swing2LimitDegrees=15, twistLimitDegrees=2}
]
