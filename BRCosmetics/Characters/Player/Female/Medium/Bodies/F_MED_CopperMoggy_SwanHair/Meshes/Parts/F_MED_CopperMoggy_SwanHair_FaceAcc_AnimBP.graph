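# Post-copy animation blueprint for the face accessory part of the CopperMoggy SwanHair character, adding secondary motion to its dynamic bones.
# It copies the attached parent mesh's pose, runs RigidBody simulation and a spring-driven AnimDynamics on one hair bone, then caches the result.
# In the front end, the owner's FrontEnd_Idle_Bypass curve sets how much DefaultSlot animation replaces that simulation on collar and glasses bones.
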AnimBlueprint /BRCosmetics/Characters/Player/Female/Medium/Bodies/F_MED_CopperMoggy_SwanHair/Meshes/Parts/F_MED_CopperMoggy_SwanHair_FaceAcc_AnimBP
target_skeleton=&/F_MED_CopperMoggy_SwanHair_FaceAcc_Skeleton
graphs [
 AnimGraph {nodes [
   n1 = Layered blend per bone(
    BasePose=Use cached pose 'dynPose'()
    BlendPoses_0=Slot 'DefaultSlot'()
    BlendWeights_0=Get Frontend_Idle_Bypass_IsPlayingFloat() -> (Frontend_Idle_Bypass_IsPlayingFloat:double)
   ) {layerSetup [{branchFilters [
       {boneName=dyn_L_frontCollar_master}
       {boneName=dyn_L_frontCollar_1}
       {boneName=dyn_R_frontCollar_master}
       {boneName=dyn_R_frontCollar_1}
       {boneName=dyn_L_midCollar_master}
       {boneName=dyn_L_midCollar_1}
       {boneName=dyn_R_midCollar_master}
       {boneName=dyn_R_midCollar_1}
       {boneName=MSTR_glasses}
      ]}]}
   Output Pose(Result=@n1)
   n2 = RigidBody(ComponentPose=Local To Component(LocalPose=Copy Pose From Mesh() {bUseAttachedParent=true})) {
    bClampLinearTranslationLimitToRefPose=true
    bOverrideWorldGravity=true
    overridePhysicsAsset=&/F_MED_CopperMoggy_SwanHair_FaceAcc_Physics
    alphaBoolBlend {blendOption=Linear}
    alphaScaleBiasClamp {interpSpeedDecreasing=3, interpSpeedIncreasing=6}
    baseBoneRef {boneName=head}
    overrideWorldGravity {z=-980}
    simSpaceSettings {
     dampingAlpha=0
     maxAngularAcceleration=120
     maxAngularVelocity=24
     maxLinearAcceleration=4800
     maxLinearVelocity=450
     worldAlpha=1
     externalLinearDragV=<1 1 1>
    }
   }
   n3 = Anim Dynamics - Bone: dyn_hair_07_b(ComponentPose=@n2, Alpha=1, ExternalForce="0,0,0") {
    angularDampingOverride=1
    angularSpringConstant=12000
    bAngularSpring=true
    bOverrideAngularDamping=true
    bUseGravityOverride=true
    gravityScale=0.35
    simulationSpace=RootRelative
    boundBone {boneName=dyn_hair_07_b}
    componentLinearAccScale=<1 1 1>
    componentLinearVelScale=<1 1 1>
    gravityOverride {z=-980}
    physicsBodyDefinitions [{
      boundBone {boneName=dyn_hair_07_b}
      boxExtents=<25 25 25>
      constraintSetup {angularLimitsMax=<0 15 5>, angularLimitsMin=<0 -15 -5>, angularTarget=<1 0 0>}
      localJointOffset=<2 0 0>
     }]
   }
   Save cached pose 'dynPose'(Pose=Component To Local(ComponentPose=@n3)) {cacheName=dynPose}
  ]}
 EventGraph {nodes [
   Event BlueprintUpdateAnimation()
   > Branch(Condition=Get bIsInFrontEnd())
   > Set Frontend_Idle_Bypass_IsPlayingFloat(Frontend_Idle_Bypass_IsPlayingFloat:double=GetCurveValue(
     self=GetOwnerAnimInstance!(self:Custom Character Part Anim Instance Object Reference) -> (ReturnValue:Anim Instance Object Reference)
     CurveName=FrontEnd_Idle_Bypass
    )) -> (Output_Get:double)
  ]}
]
variables [Frontend_Idle_Bypass_IsPlayingFloat {type=number}]
